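; Gameplay ability for the ButterCake NPC that translates Berserk and Lured status tags into pawn state.
; On activation it commits and waits for either tag to be added.
; Berserk swaps the faction from Neutral to Hostile, sets replicated IsBerserk and triggers enter abilities by tag; removal reverses this and triggers exit abilities.
; Lured only toggles IsLured; each mode registers its tag-removal listener once through DoOnce.
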(event Ability|EventFailedToActivatePassiveAbility
  (params (ActorInfo :type Gameplay%20Ability%20Actor%20Info%20Structure)))

(event Ability|EventActivateAbility
  (Ability|CommitAbility)
  (Ability|Tasks|WaitGameplayTagAdd "(TagName=\"Athena.Gameplay.Status.Berserk\")"
    (:then
      (Ability|Tasks|WaitGameplayTagAdd "(TagName=\"Athena.Gameplay.Status.Lured\")"
        (:Added
          (CallFunction|OnEnterLuredMode))))
    (:Added
      (CallFunction|OnEnterBerserkMode))))

(event Ability|EventOnEndAbility
  (params (bWasCancelled :type Boolean)))

(event Custom|OnEnterBerserkMode
  (bind _returnvalue (Ability|GetAvatarActorfromActorInfo))
  ;[1 1 1] OnEnter Berserk Mode
  ;;[1 1 1] Manage Factions
  (Affiliation|RemovefromFaction _returnvalue "(TagName=\"Athena.Faction.Generic.Neutral\")")
  (Affiliation|AddtoFaction _returnvalue "(TagName=\"Athena.Faction.Generic.Hostile\")")
  ;;[1 1 1] Set IsBerserk on pawn (Replicated)
  (bind _asnpc_pawn_butter_cake_base (Utilities|Casting|CastToNPC_Pawn_ButterCake_Base _returnvalue)
    (:then
      (Variables|/ButterCake/Pawns/NPC_Pawn_ButterCake_Base.NPC_Pawn_ButterCake_Base_C|SetIsBerserk true _asnpc_pawn_butter_cake_base))
    (:CastFailed
      _))
  (Abilities|TryActivateAbilitiesbyTag (Ability|GetAbilitySystemComponent _returnvalue) (Variables|Default|GetGameplayTagEnterBerserk) true)
  ;;[1 1 1] Listen for Tag Remove
  (Utilities|FlowControl|DoOnce false
    (:Completed
      (Ability|Tasks|WaitGameplayTagRemove "(TagName=\"Athena.Gameplay.Status.Berserk\")"
        (:Removed
          (CallFunction|OnExitBerserkMode))))))

(event Custom|OnExitBerserkMode
  (bind _returnvalue (Ability|GetAvatarActorfromActorInfo))
  ;[1 1 1] OnExit Berserk Mode
  ;;[1 1 1] Manage Factions
  (Affiliation|RemovefromFaction _returnvalue "(TagName=\"Athena.Faction.Generic.Hostile\")")
  (Affiliation|AddtoFaction _returnvalue "(TagName=\"Athena.Faction.Generic.Neutral\")")
  ;;[1 1 1] Set IsBerserk on pawn (Replicated)
  (bind _asnpc_pawn_butter_cake_base (Utilities|Casting|CastToNPC_Pawn_ButterCake_Base _returnvalue)
    (:then
      (Variables|/ButterCake/Pawns/NPC_Pawn_ButterCake_Base.NPC_Pawn_ButterCake_Base_C|SetIsBerserk false _asnpc_pawn_butter_cake_base))
    (:CastFailed
      _))
  (Abilities|TryActivateAbilitiesbyTag (Ability|GetAbilitySystemComponent _returnvalue) (Variables|Default|GetGameplayTagExitBerserk) true))

(event Custom|OnEnterLuredMode
  ;[1 1 1] OnEnter Lured Mode
  (bind _asnpc_pawn_butter_cake_base (Utilities|Casting|CastToNPC_Pawn_ButterCake_Base (Ability|GetAvatarActorfromActorInfo)))
  (Variables|/ButterCake/Pawns/NPC_Pawn_ButterCake_Base.NPC_Pawn_ButterCake_Base_C|SetIsLured true _asnpc_pawn_butter_cake_base)
  (Utilities|FlowControl|DoOnce false
    (:Completed
      (Ability|Tasks|WaitGameplayTagRemove "(TagName=\"Athena.Gameplay.Status.Lured\")"
        (:Removed
          (CallFunction|OnExitLuredMode))))))

(event Custom|OnExitLuredMode
  ;[1 1 1] OnExit Lured mode
  (bind _asnpc_pawn_butter_cake_base (Utilities|Casting|CastToNPC_Pawn_ButterCake_Base (Ability|GetAvatarActorfromActorInfo)))
  (Variables|/ButterCake/Pawns/NPC_Pawn_ButterCake_Base.NPC_Pawn_ButterCake_Base_C|SetIsLured false _asnpc_pawn_butter_cake_base))
