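; Gameplay ability that auto-heals the ButterCake AI pawn when damage leaves its health low.
; On activation it casts the avatar to FortAIPawn and binds to OnDamaged, ending immediately if the cast fails.
; Each damage event applies the GE_ButterCake_AutoHeal effect to the owner while health is below the HealthThresholdToHeal variable.
; Ending the ability unbinds all events from the pawn's OnDamaged dispatcher, not just its own.
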(event Ability|EventActivateAbility
  (bind _asfort_aipawn (Utilities|Casting|CastToFortAIPawn (Ability|GetAvatarActorfromActorInfo))
    (:then
      (EventDispatchers|BindEventtoOnDamaged _asfort_aipawn (delegate Custom|OnDamaged)))
    (:CastFailed
      (Ability|EndAbility))))

(event Custom|OnDamaged
  (params (DamagedActor :type Actor%20Object%20Reference) (Damage :type Float%20%28single-precision%29) (InstigatedBy :type Controller%20Object%20Reference) (DamageCauser :type Actor%20Object%20Reference) (HitLocation :type Vector) (FHitComponent :type Primitive%20Component%20Object%20Reference) (BoneName :type Name) (Momentum :type Vector))
  (if (< (FortPawn|GetHealth (stale Ability|EventActivateAbility _asfort_aipawn)) (Variables|Default|GetHealthThresholdToHeal))
    (Ability|ApplyGameplayEffectToOwner "/ButterCake/Abilities/GE_ButterCake_AutoHeal.GE_ButterCake_AutoHeal_C")))

(event Ability|EventOnEndAbility
  (params (bWasCancelled :type Boolean))
  (bind _asfort_aipawn (Utilities|Casting|CastToFortAIPawn (Ability|GetAvatarActorfromActorInfo)))
  (EventDispatchers|UnbindallEventsfromOnDamaged _asfort_aipawn))
