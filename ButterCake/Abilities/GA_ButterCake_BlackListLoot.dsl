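; ButterCake gameplay ability that blacklists a loot pickup for the AI.
; On activation it commits, then casts the AI ability target to a FortPickup.
; On success it adds the blacklist gameplay tag to that pickup through the AI loot service, waits briefly, and ends.
; A failed cast ends the ability immediately.
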(event Ability|EventActivateAbility
  (Ability|CommitAbility)
  (bind _asfort_pickup (Utilities|Casting|CastToFortPickup (Ability|GetAIAbilityTarget))
    (:then
      (bind _returnvalue (AI|GetAIServiceLoot))
      (AI|K2AddGameplayTagtoLoot _returnvalue _asfort_pickup (Variables|Default|GetBlackListGameplayTag))
      (Utilities|FlowControl|Delay 2.0)
      (Ability|EndAbility))
    (:CastFailed
      (Ability|EndAbility))))

(event Ability|EventOnEndAbility
  (params (bWasCancelled :type Boolean)))
