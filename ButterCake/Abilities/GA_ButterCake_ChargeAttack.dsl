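; Gameplay ability for the ButterCake AI pawn's charge attack, which launches and applies effects to pawns and vehicles it hits.
; It traces a forward point within its leash, or under its target in safe-charge mode, and projects it onto the navmesh.
; A montage notify starts MoveAI and self-restarting target selection, queuing hits processed one per tick and flushed on end.
; Activation requires a minimum target distance and no water; move completion triggers the outro.
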(event Custom|RestartTargetSelection
  (GameplayTasks|EndTask (Variables|Default|GetTargetSelectionTask))
  (bind (_asynctaskproxy _targetdata) (Ability|Tasks|DoTargetSelectionContinuous "None" "(TagName=\"Targeting.ApplicationTag.0\")")
    (:then
      (Variables|Default|SetTargetSelectionTask _asynctaskproxy))
    (:Targeted
      (for _array_element (Ability|TargetData|GetAllActorsfromTargetData _targetdata)
        (Utilities|Array|AddUnique (Variables|Default|GetChargeTargetsToProcess) _array_element))
      (CallFunction|RestartTargetSelection))))

(event Ability|EventActivateAbility
  (bind _buttercakepawn (Variables|Default|GetButterCakePawn))
  (bind _returnvalue (Fortnite|AI|GetNavigationDataforActor _buttercakepawn))
  (bind _montageinfo_montagetoplay (Utilities|Struct|BreakFortGameplayAbilityMontageInfo (Variables|GetMontageInfo_MontageToPlay)))
  (Ability|CommitAbilityCost false)
  (Utilities|Array|Clear (Variables|Default|GetChargeTargetsToProcess))
  (bind _asfort_aipawn (Utilities|Casting|CastToFortAIPawn (Ability|GetActivatingPawn))
    (:then
      (Variables|Default|SetButterCakePawn _asfort_aipawn)
      (Variables|Default|SetOutroIsPlaying false)
      (bind (_chargedestinationtracestart _chargedestinationtraceend _isvalid) (CallFunction|GetDestinationLocation))
      (if _isvalid
        (bind (_outhit _returnvalue_2) (Collision|LineTraceByChannel :Start _chargedestinationtracestart :End _chargedestinationtraceend :TraceChannel "TraceTypeQuery7" :bTraceComplex false :DrawDebugType "None" :bIgnoreSelf true :TraceColor "(R=1.000000,G=0.000000,B=0.000000,A=1.000000)" :TraceHitColor "(R=0.000000,G=1.000000,B=0.000000,A=1.000000)" :DrawTime 1.8e+02))
        (if _returnvalue_2
          (if (.ReturnValue (AI|Navigation|ProjectPointtoNavigation (.Location (Collision|BreakHitResult _outhit)) _returnvalue "/ButterCake/NavFilters/NavFilter_ButterCake.NavFilter_ButterCake_C" "512.000000,512.000000,512.000000"))
            (Utilities|Time|SetTimerforNextTickbyEvent (EventDispatchers|CreateEvent :!event Custom|OnProcessTargetsEvent))
            (Ability|Tasks|PlayMontageAndWaitNotify "BCChargeAttack" _montageinfo_montagetoplay
              (:OnInterrupted
                (if (Variables|Default|GetOutroIsPlaying)
                  _
                  (else
                    _)))
              (:OnCancelled
                _)
              (:OnNotifyBegin
                (bind _asynctaskproxy (Ability|Tasks|MoveAI "MoveToChargeAttack" 0 (AI|Navigation|ProjectPointtoNavigation (.Location (Collision|BreakHitResult _outhit)) _returnvalue "/ButterCake/NavFilters/NavFilter_ButterCake.NavFilter_ButterCake_C" "512.000000,512.000000,512.000000") "UsePathfinding" "Ignore" false "Enable" false true 2e+02 0 (Variables|Default|GetNavFilterClass) "(TagName=\"\")"))
                (Utilities|IsValid _asynctaskproxy
                  (:Is%20Valid
                    (bind _output_get (Variables|Default|SetMoveTask _asynctaskproxy))
                    (EventDispatchers|BindEventtoOnMoveFinished _output_get (delegate Custom|ButterCakeMoveFinished))
                    ;[1 1 1] Play Loop VFX
                    (Ability|AddGameplayCueToOwner (Variables|Default|GetChargeLoopGCTag))
                    (bind (_asynctaskproxy_2 _targetdata) (Ability|Tasks|DoTargetSelectionContinuous "None" "(TagName=\"Targeting.ApplicationTag.0\")")
                      (:then
                        (Variables|Default|SetTargetSelectionTask _asynctaskproxy_2))
                      (:Targeted
                        (for _array_element (Ability|TargetData|GetAllActorsfromTargetData _targetdata)
                          (Utilities|Array|AddUnique (Variables|Default|GetChargeTargetsToProcess) _array_element))
                        (CallFunction|RestartTargetSelection))))
                  (:Is%20Not%20Valid
                    (Ability|EndAbility)))))
            (Ability|EndAbility)
            (else
              _))
          (else
            _))
        (Ability|EndAbility)
        (else
          _)))
    (:CastFailed
      (Ability|EndAbility))))

(event Custom|ButterCakeMoveFinished
  (params (Result :type EPathFollowingResult%20Enum :default "Success") (AIController :type AIController%20Object%20Reference))
  ;[1 1 1] Play Charge Outro
  (Ability|RemoveGameplayCueFromOwner (Variables|Default|GetChargeLoopGCTag))
  (Variables|Default|SetOutroIsPlaying true)
  (Ability|Tasks|PlayMontageAndWaitNotify "BCChargeAttackFinish" (Variables|Default|GetOutroMontage)
    (:OnCompleted
      _)
    (:OnInterrupted
      _)
    (:OnCancelled
      _))
  (Ability|EndAbility))

(event Custom|OnProcessTargetsEvent
  ;[1 1 1] On Process Targets
  (if (not (Utilities|Array|IsEmpty (Variables|Default|GetChargeTargetsToProcess)))
    (CallFunction|ProcessTarget (Utilities|Array|Get(acopy) (Variables|Default|GetChargeTargetsToProcess)))
    (Utilities|Array|RemoveIndex (Variables|Default|GetChargeTargetsToProcess)))
  (Utilities|Time|SetTimerforNextTickbyEvent (EventDispatchers|CreateEvent :!event Custom|OnProcessTargetsEvent)))

(event Ability|EventOnEndAbility
  (params (bWasCancelled :type Boolean))
  (Ability|CommitAbilityCooldown false true)
  (Utilities|FlowControl|Sequence
    (:then_0
      ;[1 1 1] Process the rest of targets if we didn't have the time to do it during the GA life time
      (for _array_element (Variables|Default|GetChargeTargetsToProcess)
        (CallFunction|ProcessTarget _array_element)))
    (:then_1
      (GameplayTasks|EndTask (Variables|Default|GetTargetSelectionTask))
      (bind _movetask (Variables|Default|GetMoveTask)
        (:then
          (EventDispatchers|UnbindallEventsfromOnMoveFinished _movetask)
          (Variables|Default|SetMoveTask 0))
        (:else)))))

(fn GetCustomAbilitySourceTransform
  (return (Math|Transform|MakeTransform (+ (- (Transformation|GetSocketLocation (Actor|GetComponentbyClass (Ability|GetAvatarActorfromActorInfo) "/Script/FortniteGame.FortPooledSkelMeshComponent") (Variables|Default|GetChargeAttackSocketName)) (* (Transformation|GetActorForwardVector (Ability|GetAvatarActorfromActorInfo)) (ScalableFloat|GetValueatLevel (Variables|Default|GetSF_DetectionOffset)))) "0, 0,-200.000000"))))

(fn GetDestinationLocation
  (bind _buttercakepawn (Variables|Default|GetButterCakePawn))
  (bind _returnvalue (Transformation|GetActorLocation _buttercakepawn))
  (bind _returnvalue_2 (Transformation|GetActorForwardVector _buttercakepawn))
  (bind _sf_chargedistance (Variables|Default|GetSF_ChargeDistance))
  (bind _returnvalue_3 (ScalableFloat|GetValueatLevel _sf_chargedistance))
  (bind _returnvalue_4 (Pawn|GetController _buttercakepawn))
  (bind _returnvalue_5 (Actor|GetComponentbyClass _returnvalue_4 "/Script/FortniteAI.FortAthenaLeashComponent"))
  (bind _returnvalue_6 (* _returnvalue_2 "1.000000,1.000000,0"))
  (bind _returnvalue_7 (* _returnvalue_6 _returnvalue_3))
  (bind _returnvalue_8 (+ _returnvalue _returnvalue_7))
  (if (not (ScalableFloat|GetValueasBool (Variables|Default|GetSF_UseSafeCharge)))
    (Utilities|IsValid _returnvalue_5
      (:Is%20Valid
        (if (Class|FortAthenaLeashComponent|IsInsideLeash _returnvalue_5 _returnvalue_8 0.0)
          (return (Utilities|Operators|Add _returnvalue_8 (Math|Vector|MakeVector 0.0 0.0 (ScalableFloat|GetValueatLevel (Variables|Default|GetSF_ChargeTraceValues)))) (Utilities|Operators|Subtract _returnvalue_8 (Math|Vector|MakeVector 0.0 0.0 (ScalableFloat|GetValueatLevel (Variables|Default|GetSF_ChargeTraceValues)))) true)
          (else
            (return "0, 0, 0" "0, 0, 0" false))))
      (:Is%20Not%20Valid
        _)))
  (return (+ (Transformation|GetActorLocation (Ability|GetAIAbilityTarget)) "0, 0,50.000000") (+ (Transformation|GetActorLocation (Ability|GetAIAbilityTarget)) "0, 0,-2000.000000") true))

(fn CanActivateAbility
  (params (ActorInfo :type Gameplay%20Ability%20Actor%20Info%20Structure) (Handle :type Gameplay%20Ability%20Spec%20Handle%20Structure))
  (bind _returnvalue (Ability|GetAIAbilityTarget))
  (Utilities|IsValid _returnvalue
    (:Is%20Valid
      (bind _asfort_aipawn (Utilities|Casting|CastToFortAIPawn (Ability|GetAvatarActorfromActorInfo))
        (:then
          (return 0 (and (>= (Math|Vector|Distance(Vector) (Transformation|GetActorLocation _returnvalue) (Transformation|GetActorLocation (Ability|GetAvatarActorfromActorInfo))) (select (ScalableFloat|GetValueasBool (Variables|Default|GetSF_UseSafeCharge)) (ScalableFloat|GetValueatLevel (Variables|Default|GetSF_MinDistanceToSafeCharge)) (ScalableFloat|GetValueatLevel (Variables|Default|GetSF_MinDistanceToCharge)))) (not (Utilities|IsValid (WaterBody|GetCurrentWaterBody _asfort_aipawn) :!class /Script/Engine.KismetSystemLibrary)))))
        (:CastFailed
          _)))
    (:Is%20Not%20Valid
      _))
  (return 0 false))

(fn ProcessTarget
  (params (ActorToProcess :type Actor%20Object%20Reference))
  (Utilities|IsValid ActorToProcess
    (:Is%20Valid
      (bind _asfort_pawn (Utilities|Casting|CastToFortPawn (Variables|GetActorToProcess))
        (:then
          (Utilities|IsinVehicle _asfort_pawn
            (:false
              (Character|LaunchCharacter _asfort_pawn (+ (Math|Vector|vector+vector (* (Transformation|GetActorForwardVector (Variables|Default|GetButterCakePawn)) (ScalableFloat|GetValueatLevel (Variables|Default|GetSF_PlayerLaunchVelocity))) (Math|Vector|MakeVector 0.0 0.0 (ScalableFloat|GetValueatLevel (Variables|Default|GetSF_PlayerLaunchVelocityAddedZ)))) (Transformation|GetActorForwardVector (Variables|Default|GetButterCakePawn))) true true))))
        (:CastFailed
          (bind _asfort_athena_vehicle (Utilities|Casting|CastToFortAthenaVehicle (Variables|GetActorToProcess)))
          (Audio|BroadcastAbilityImpactatLocation 0 _asfort_athena_vehicle (+ (Math|Vector|vector+vector (* (Transformation|GetActorForwardVector (Variables|Default|GetButterCakePawn)) (ScalableFloat|GetValueatLevel (Variables|Default|GetSF_VehicleLaunchVelocity))) (Math|Vector|MakeVector 0.0 0.0 (ScalableFloat|GetValueatLevel (Variables|Default|GetSF_VehicleLaunchVelocityAddedZ)))) (Transformation|GetActorForwardVector (Variables|Default|GetButterCakePawn))) 777.0)))
      (Ability|ApplyGameplayEffectContainer (Ability|TargetData|AbilityTargetDatafromHitResult (Collision|MakeHitResult false false 0.0 0.0 (Transformation|GetActorLocation (Variables|GetActorToProcess)) (Transformation|GetActorLocation (Variables|GetActorToProcess)) "0,0,1" "0,0,1" 0 (Variables|GetActorToProcess))) "(TagName=\"Targeting.ApplicationTag.0\")")
      (GameplayCue|ExecuteGameplayCue (Variables|GetActorToProcess) (Variables|Default|GetGameplayCueTagHitTarget)))))
