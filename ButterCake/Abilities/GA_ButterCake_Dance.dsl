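; ButterCake gameplay ability that plays a dance montage.
; On activation it commits the ability cost and plays the montage and rate from its montage info variable through a PlayMontageAndWaitNotify task.
; The ability ends when the montage completes, is interrupted or cancelled, or when the task proxy is invalid.
; Cooldown is committed in OnEndAbility rather than at activation, whether or not the ability was cancelled.
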(event Ability|EventActivateAbility
  (bind (_montageinfo_montagetoplay _montageinfo_animplayrate) (Utilities|Struct|BreakFortGameplayAbilityMontageInfo (Variables|GetMontageInfo_MontageToPlay)))
  (Ability|CommitAbilityCost false)
  (bind _asynctaskproxy (Ability|Tasks|PlayMontageAndWaitNotify "BCDance" _montageinfo_montagetoplay _montageinfo_animplayrate)
    (:then
      (Utilities|IsValid _asynctaskproxy
        (:Is%20Not%20Valid
          _)))
    (:OnCompleted
      _)
    (:OnInterrupted
      _)
    (:OnCancelled
      _))
  (Ability|EndAbility))

(event Ability|EventOnEndAbility
  (params (bWasCancelled :type Boolean))
  (Ability|CommitAbilityCooldown false))
