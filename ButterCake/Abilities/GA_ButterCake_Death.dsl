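; Death gameplay ability for ButterCake that tosses loot pickups into the world when the ability activates.
; After the parent activation runs, it loops over a count truncated from a scalable float.
; Each pass spawns one configured item definition, raised above the avatar's location by the capsule half height, with a randomized horizontal, upward-biased toss direction.
; The end-ability event holds only a disabled parent call.
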(event Ability|EventActivateAbility
  (bind _returnvalue (Ability|GetAvatarActorfromActorInfo))
  (|CallParentFunction)
  (for _ (range 1 (+ (Math|Float|Truncate (ScalableFloat|GetValueatLevel (Variables|Default|GetSF_AmountOfLootToSpawn))) 1))
    (bind _asfort_pawn (Utilities|Casting|CastToFortPawn _returnvalue))
    (Inventory|SpawnPickupInWorld (Variables|Default|GetItemDefinitionToSpawnOnDeath) 1 (Math|Vector|vector+vector (Transformation|GetActorLocation _returnvalue) (Math|Vector|MakeVector 0.0 0.0 (Components|Capsule|GetScaledCapsuleHalfHeight (Variables|/Script/Engine.Character|GetCapsuleComponent _asfort_pawn)))) (Math|Vector|MakeVector (Math|Random|RandomFloatinRange -1.0 1.0) (Math|Random|RandomFloatinRange -1.0 1.0) 1.0) 0 true true false -1 "Tossed" "Unset" 0 false)))

(event Ability|EventOnEndAbility
  (params (bWasCancelled :type Boolean))
  (|CallParentFunction bWasCancelled) ; This node is disabled and will not be called.\nDrag off pins to build functionality.
)
