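; Gameplay ability that plays ButterCake's enter-berserk montage and throws off pawns riding on the creature.
; A montage notify triggers EjectPawn, which finds pawns using the avatar as a base and skips any beyond a maximum distance.
; Each remaining pawn with an ability system receives fall damage immunity and is launched along the avatar's up vector, tilted about its right axis.
; A delay based on montage length provides a timed ability end.
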(event Ability|EventActivateAbility
  (bind (_montageinfo_montagetoplay _montageinfo_animplayrate) (Utilities|Struct|BreakFortGameplayAbilityMontageInfo (Variables|GetMontageInfo_MontageToPlay)))
  (Ability|CommitAbility)
  (bind _asynctaskproxy (Ability|Tasks|PlayMontageAndWaitNotify "EnterBerserk" _montageinfo_montagetoplay _montageinfo_animplayrate)
    (:then
      (Utilities|IsValid _asynctaskproxy
        (:Is%20Valid
          (Utilities|FlowControl|Delay (+ (Animation|GetPlayLength _montageinfo_montagetoplay) 1.0))
          (Ability|EndAbility))
        (:Is%20Not%20Valid
          _)))
    (:OnCompleted
      _)
    (:OnInterrupted
      _)
    (:OnCancelled
      _)
    (:OnNotifyBegin
      (CallFunction|EjectPawn)))
  (Ability|EndAbility))

(event Ability|EventOnEndAbility
  (params (bWasCancelled :type Boolean)))

(fn EjectPawn
  (bind _returnvalue (Ability|GetAvatarActorfromActorInfo))
  (bind _asfort_aipawn (Utilities|Casting|CastToFortAIPawn _returnvalue))
  (bind _returnvalue_2 (FortAIPawn|FindValidPawnsUsedorUsingThisasABase _asfort_aipawn))
  (for _array_element _returnvalue_2
    (bind _asfort_pawn (Utilities|Casting|CastToFortPawn _array_element))
    ;[1 1 1] Don't apply if player is too far but still registered as on BC
    (if (< (Math|Vector|Distance(Vector) (Transformation|GetActorLocation _asfort_pawn) (Transformation|GetActorLocation _asfort_aipawn)) (ScalableFloat|GetValueatLevel (LocalVariables|GetPlayerEjectMaxDistance)))
      (Utilities|IsValid (Ability|GetAbilitySystemComponent _asfort_pawn)
        (:Is%20Valid
          (GameplayEffects|ApplyGameplayEffectToTarget :self (Ability|GetAbilitySystemComponent _asfort_pawn) :GameplayEffectClass (LocalVariables|GetGE_FallDamageImmunity) :Target (Ability|GetAbilitySystemComponent _asfort_pawn) :Level 0.0)
          (Character|LaunchCharacter _asfort_pawn (* (Math|Vector|RotateVector (Transformation|GetActorUpVector _returnvalue) (Math|Vector|RotatorfromAxisandAngle (Transformation|GetActorRightVector _returnvalue) -3e+01)) (ScalableFloat|GetValueatLevel (LocalVariables|GetLaunchSpeed))) true true))))))
