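; GA_ButterCake_EnterLuring is a gameplay ability that plays ButterCake's montage for entering the luring state.
; It commits, then plays the montage from its montage info at the configured rate through a PlayMontageAndWaitNotify task.
; The ability ends when the montage completes, is interrupted or cancelled, or the task is invalid.
; A fallback delay of the montage's unscaled length plus one second also ends it.
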(event Ability|EventActivateAbility
  (bind (_montageinfo_montagetoplay _montageinfo_animplayrate) (Utilities|Struct|BreakFortGameplayAbilityMontageInfo (Variables|GetMontageInfo_MontageToPlay)))
  (Ability|CommitAbility)
  (bind _asynctaskproxy (Ability|Tasks|PlayMontageAndWaitNotify "EnterLuring" _montageinfo_montagetoplay _montageinfo_animplayrate)
    (:then
      (Utilities|IsValid _asynctaskproxy
        (:Is%20Valid
          (Utilities|FlowControl|Delay (+ (Animation|GetPlayLength _montageinfo_montagetoplay) 1.0))
          (Ability|EndAbility))
        (:Is%20Not%20Valid
          _)))
    (:OnCompleted
      _)
    (:OnInterrupted
      _)
    (:OnCancelled
      _))
  (Ability|EndAbility))
