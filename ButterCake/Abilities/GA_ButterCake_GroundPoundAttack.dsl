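; Gameplay ability for the ButterCake AI's ground pound, which launches riders off its back and knocks away nearby targets.
; A montage notify traces for water or ground, plays the matching impact cue and runs target selection.
; Targets are hit one per frame for performance, with leftovers flushed on ability end as cooldown commits.
; Pawns, vehicles and physics objects get launched away; valid building actors receive the effect container.
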(event Ability|EventActivateAbility
  (bind _groundpoundimpactlocation (Variables|Default|GetGroundPoundImpactLocation))
  (bind _returnvalue (Ability|GameplayCue|MakeGameplayCueParameters :NormalizedMagnitude 0.0 :RawMagnitude 0.0 :Location _groundpoundimpactlocation :Normal "0,0,0" :GameplayEffectLevel 1 :AbilityLevel 1 :bReplicateLocationWhenUsingMinimalRepProxy false))
  (bind (_montageinfo_montagetoplay _montageinfo_animplayrate) (Utilities|Struct|BreakFortGameplayAbilityMontageInfo (Variables|GetMontageInfo_MontageToPlay)))
  (Ability|CommitAbilityCost false)
  (Utilities|Array|Clear (Variables|Default|GetGroundPoundTargets))
  ;[1 1 1] Play GroundPound Animation
  (Ability|Tasks|PlayMontageAndWaitNotify "MontageGroundPound" _montageinfo_montagetoplay _montageinfo_animplayrate
    (:then
      (Default|EjectPawn))
    (:OnCompleted
      _)
    (:OnInterrupted
      _)
    (:OnCancelled
      _)
    (:OnNotifyBegin
      ;[1 1 1] Compute Impact Location
      (bind (_isresultvalid _groundpoundlocation _isinwater) (CallFunction|FindGroundPoundLocation))
      (Variables|Default|SetGroundPoundImpactLocation _groundpoundlocation)
      (if _isresultvalid
        ;[1 1 1] Play VFX
        (if _isinwater
          (Ability|ExecuteGameplayCueWithParamsOnOwner (Variables|Default|GetGroundPoundImpactInWaterCueTag) _returnvalue)
          (else
            (Ability|ExecuteGameplayCueWithParamsOnOwner (Variables|Default|GetGroundPoundImpactCueTag) _returnvalue)))
        (Variables|Default|SetButterCakeLocation (Transformation|GetActorLocation (Ability|GetAvatarActorfromActorInfo)))
        ;[1 1 1] Apply Effect on Targets
        (bind (_asynctaskproxy _targetdata) (Ability|Tasks|DoTargetSelection "None" "(TagName=\"Targeting.ApplicationTag.0\")" true)
          (:Targeted
            (Variables|Default|SetGroundPoundTargets (Ability|TargetData|GetAllActorsfromTargetData _targetdata))
            (Variables|Default|SetCollisionReactionComp (Actor|GetComponentbyClass (Ability|GetAvatarActorfromActorInfo) "/Script/FortniteGame.FortAthenaCollisionReactionComponent"))
            (Utilities|Time|SetTimerforNextTickbyEvent (delegate Custom|ProcessTargets))))
        (else
          (Ability|EndAbility)))))
  (Ability|EndAbility))

(event Custom|ProcessTargets
  (bind _groundpoundtargets (Variables|Default|GetGroundPoundTargets))
  (bind _returnvalue (Utilities|Array|Length _groundpoundtargets))
  (bind _returnvalue_2 (- _returnvalue 1))
  ;[1 1 1] Process 1 Target by Frame for Perf
  (if (Utilities|Array|IsNotEmpty (Variables|Default|GetGroundPoundTargets))
    (CallFunction|HitTarget (Utilities|Array|Get(acopy) _groundpoundtargets _returnvalue_2))
    (Utilities|Array|RemoveIndex (Variables|Default|GetGroundPoundTargets) _returnvalue_2)
    (Utilities|Time|SetTimerforNextTickbyEvent (delegate Custom|ProcessTargets))))

(event Ability|EventOnEndAbility
  (params (bWasCancelled :type Boolean))
  (Ability|CommitAbilityCooldown false true)
  (for _array_element (Variables|Default|GetGroundPoundTargets)
    (CallFunction|HitTarget _array_element)))

(event Custom|HitTarget
  (params (TargetActor :type Actor%20Object%20Reference))
  (bind _collisionreactioncomp (Variables|Default|GetCollisionReactionComp))
  ;[1 1 1] Hit Target
  (Utilities|IsValid TargetActor
    (:Is%20Valid
      (bind _asfort_pawn (Utilities|Casting|CastToFortPawn TargetActor)
        (:then
          (CallFunction|LaunchPawn _asfort_pawn))
        (:CastFailed
          (bind _asfort_athena_vehicle (Utilities|Casting|CastToFortAthenaVehicle TargetActor)
            (:then
              (CallFunction|LunchVehicle _asfort_athena_vehicle)
              (GameplayCue|ExecuteGameplayCue TargetActor (Variables|Default|GetGameplayCueTagOnHit)))
            (:CastFailed
              (bind _asbuilding_actor (Utilities|Casting|CastToBuildingActor TargetActor)
                (:then
                  (Utilities|IsValid _collisionreactioncomp
                    (:Is%20Valid
                      ;;[1 1 1] If not valid for reaction, just ignore it
                      (if (CollisionReactions|K2IsBuildingActorValidforReaction _collisionreactioncomp _asbuilding_actor)
                        (if (ScalableFloat|GetValueasBool (Variables|Default|GetLaunchPhysObjectEnable))
                          (bind _returnvalue (FortPhysicsSimulationLibrary|CanActorSimulatePhysics _asbuilding_actor))
                          (if _returnvalue
                            (CallFunction|LaunchPhysActor _asbuilding_actor)
                            (else
                              _))
                          (else
                            _))
                        (Ability|ApplyGameplayEffectContainer (Ability|TargetData|AbilityTargetDatafromHitResult (Collision|MakeHitResult false false 0.0 0.0 (Transformation|GetActorLocation TargetActor) (Transformation|GetActorLocation TargetActor) "0,0,1" "0,0,1" 0 TargetActor)) "(TagName=\"Targeting.ApplicationTag.0\")")))
                    (:Is%20Not%20Valid
                      _)))
                (:CastFailed
                  _)))))))))

(event Custom|LaunchPawn
  (params (FortPawn :type Fort%20Pawn%20Object%20Reference))
  ;[1 1 1] Apply Effects on Pawn
  (Utilities|IsinVehicle FortPawn
    (:false
      (bind _launchvelocity (CallFunction|ComputeLaunchVelocity (Variables|Default|GetButterCakeLocation) (Transformation|GetActorLocation FortPawn)))
      (Character|LaunchCharacter FortPawn _launchvelocity true true))))

(event Custom|LunchVehicle
  (params (FortAthenaVehicle :type Fort%20Athena%20Vehicle%20Object%20Reference))
  ;[1 1 1] Apply Effects on Vehcile
  (bind _launchvelocity (CallFunction|ComputeLaunchVelocity (Variables|Default|GetButterCakeLocation) (Transformation|GetActorLocation FortAthenaVehicle) true))
  (Audio|BroadcastAbilityImpactatLocation 0 FortAthenaVehicle _launchvelocity 777.0))

(event Custom|Launch%20Phys%20Actor
  (params (Phys%20Actor :type Actor%20Object%20Reference))
  ;[1 1 1] Apply Launch on Phys Objects
  (PhysicsObject|WakeUpPhysicsObject Phys%20Actor)
  (bind _launchvelocity (CallFunction|ComputePhysActorLaunchVelocity (Variables|Default|GetButterCakeLocation) (Transformation|GetActorLocation Phys%20Actor)))
  (FortPhysicsSimulationLibrary|SetPhysicsLinearVelocity Phys%20Actor _launchvelocity false "None" 0 "(MaxSimSizeAllowed=Large,bRejectEmbeddedInLandscape=False)")
  (FortPhysicsSimulationLibrary|SetPhysicsAngularVelocity Phys%20Actor (Math|Vector|MakeVector (Math|Random|RandomFloatinRange (Variables|Default|GetMinAngularVelocity) (Variables|Default|GetMaxAngularVelocity)) (Math|Random|RandomFloatinRange (Variables|Default|GetMinAngularVelocity) (Variables|Default|GetMaxAngularVelocity)) (Math|Random|RandomFloatinRange (Variables|Default|GetMinAngularVelocity) (Variables|Default|GetMaxAngularVelocity))) false "None" 0 "(MaxSimSizeAllowed=Large,bRejectEmbeddedInLandscape=False)"))

(fn ComputeLaunchVelocity
  (params (ImpactLocation :type Vector) (ActorLocation :type Vector) (IsVehicle :type Boolean))
  (bind (_returnvalue_x _returnvalue_y) (Math|Vector|BreakVector (* (Math|Vector|Normalize (- ActorLocation ImpactLocation)) (select IsVehicle (ScalableFloat|GetValueatLevel (Variables|Default|GetLaunchVehicleSpeedXY)) (ScalableFloat|GetValueatLevel (Variables|Default|GetLaunchPawnSpeedXY))))))
  (return _returnvalue_x _returnvalue_y (select IsVehicle (ScalableFloat|GetValueatLevel (Variables|Default|GetLaunchVehicleSpeedZ)) (ScalableFloat|GetValueatLevel (Variables|Default|GetLaunchPawnSpeedZ)))))

(fn ComputePhysActorLaunchVelocity
  (params (ImpactLocation :type Vector) (ActorLocation :type Vector))
  (bind (_returnvalue_x _returnvalue_y) (Math|Vector|BreakVector (* (Math|Vector|Normalize (- ActorLocation ImpactLocation)) (ScalableFloat|GetValueatLevel (Variables|Default|GetLaunchPhysObjectSpeedXY)))))
  (return _returnvalue_x _returnvalue_y (ScalableFloat|GetValueatLevel (Variables|Default|GetLaunchPhysObjectSpeedZ))))

(fn EjectPawn
  (bind _returnvalue (Ability|GetAvatarActorfromActorInfo))
  (bind _asfort_aipawn (Utilities|Casting|CastToFortAIPawn _returnvalue))
  (bind _returnvalue_2 (FortAIPawn|FindValidPawnsUsedorUsingThisasABase _asfort_aipawn))
  (for _array_element _returnvalue_2
    (bind _asfort_pawn (Utilities|Casting|CastToFortPawn _array_element))
    ;[1 1 1] Don't apply if player is too far but still registered as on BC
    (if (< (Math|Vector|Distance(Vector) (Transformation|GetActorLocation _asfort_pawn) (Transformation|GetActorLocation _asfort_aipawn)) (ScalableFloat|GetValueatLevel (LocalVariables|GetPlayerEjectMaxDistance)))
      (Utilities|IsValid (Ability|GetAbilitySystemComponent _asfort_pawn)
        (:Is%20Valid
          (GameplayEffects|ApplyGameplayEffectToTarget :self (Ability|GetAbilitySystemComponent _asfort_pawn) :GameplayEffectClass (LocalVariables|GetGE_FallDamageImmunity) :Target (Ability|GetAbilitySystemComponent _asfort_pawn) :Level 0.0)
          (Character|LaunchCharacter _asfort_pawn (* (Math|Vector|RotateVector (Transformation|GetActorUpVector _returnvalue) (Math|Vector|RotatorfromAxisandAngle (Transformation|GetActorRightVector _returnvalue) -3e+01)) (ScalableFloat|GetValueatLevel (LocalVariables|GetLaunchSpeed))) true true))))))

(fn FindGroundPoundLocation
  (bind _output_get (LocalVariables|SetStartTraceLocation (+ (+ (Transformation|GetActorLocation (Ability|GetAvatarActorfromActorInfo)) "0,0,500.000000") (* (Transformation|GetActorForwardVector (Ability|GetAvatarActorfromActorInfo)) (ScalableFloat|GetValueatLevel (Variables|Default|GetEffectForwardOffset))))))
  (LocalVariables|SetEndTraceLocation (+ _output_get "0, 0,-1500.000000"))
  ;[1 1 1] Water Check
  (bind (_outhit _returnvalue) (Collision|LineTraceByChannel (LocalVariables|GetStartTraceLocation) (LocalVariables|GetEndTraceLocation) "TraceTypeQuery12"))
  (bind (_outhit_bblockinghit _outhit_binitialoverlap _outhit_time _outhit_distance _outhit_location) (Collision|BreakHitResult _outhit))
  (if _returnvalue
    (return true _outhit_location true)
    (else
      ;[1 1 1] Ground Check
      (bind (_outhit_2 _returnvalue_2) (Collision|LineTraceByChannel (LocalVariables|GetStartTraceLocation) (LocalVariables|GetEndTraceLocation) "TraceTypeQuery10"))
      (bind (_outhit_bblockinghit_2 _outhit_binitialoverlap_2 _outhit_time_2 _outhit_distance_2 _outhit_location_2) (Collision|BreakHitResult _outhit_2))
      (return _returnvalue_2 _outhit_location_2 false))))

(fn GetCustomAbilitySourceTransform
  (return (Math|Transform|MakeTransform (Variables|Default|GetGroundPoundImpactLocation))))
