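; GA_ButterCake_Idle is a Gameplay Ability that keeps the ButterCake character idle for a random length of time.
; On activation it commits, then waits a random delay between the level-scaled MinIdleDuration and MaxIdleDuration scalable floats before ending itself.
; The commit result is not checked, and the end-ability event has no logic.
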(event Ability|EventActivateAbility
  (Ability|CommitAbility)
  (Utilities|FlowControl|Delay (Math|Random|RandomFloatinRange (ScalableFloat|GetValueatLevel (Variables|Default|GetMinIdleDuration)) (ScalableFloat|GetValueatLevel (Variables|Default|GetMaxIdleDuration))))
  (Ability|EndAbility))

(event Ability|EventOnEndAbility
  (params (bWasCancelled :type Boolean)))
