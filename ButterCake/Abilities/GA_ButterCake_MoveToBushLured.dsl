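; GA_ButterCake_MoveToBushLured is a Gameplay Ability for the ButterCake character, named for a lured move toward a bush.
; Its activate and end events only call the parent implementations, passing the cancelled flag through on end.
; The graph adds no logic of its own, so all behavior comes from the parent ability class.
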(event Ability|EventActivateAbility
  (|CallParentFunction))

(event Ability|EventOnEndAbility
  (params (bWasCancelled :type Boolean))
  (|CallParentFunction bWasCancelled))
