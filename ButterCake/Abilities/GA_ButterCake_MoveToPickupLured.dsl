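; GA_ButterCake_MoveToPickupLured is a gameplay ability for the ButterCake AI pawn that covers being lured toward a pickup.
; Activation is allowed only when the pawn's goal is a luring pickup that is either tagged as fruit or player-tossed within a short lifetime limit.
; On activation it runs the parent logic, then reports the lure to the telemetry component on the pawn's ButterCake controller.
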(event Ability|EventActivateAbility
  (|CallParentFunction)
  (CallFunction|FeedAnalytics))

(event Ability|EventOnEndAbility
  (params (bWasCancelled :type Boolean))
  (|CallParentFunction bWasCancelled))

(fn FeedAnalytics
  (bind _asfort_aipawn (Utilities|Casting|CastToFortAIPawn (Ability|GetAvatarActorfromActorInfo)))
  (bind _asbutter_cake_controller (Utilities|Casting|CastToButterCake_Controller (AI|GetAIController _asfort_aipawn)))
  (ButterCake-Telemetry|OnLured (Variables|/ButterCake/Controllers/ButterCake_Controller.ButterCake_Controller_C|GetFortButterCakeComponent_Telemetry _asbutter_cake_controller)))

(fn CanActivateAbility
  (params (ActorInfo :type Gameplay%20Ability%20Actor%20Info%20Structure) (Handle :type Gameplay%20Ability%20Spec%20Handle%20Structure))
  (bind _returnvalue (AI|AIPawnfromGAHasLuringPickupasGoal :ActorInfo ActorInfo :MaxLifeTime 1e+01 :RequiredPickupSpawnSource "TossedByPlayer"))
  (bind _returnvalue_2 (AI|AIPawnfromGAHasLuringPickupasGoal ActorInfo (Variables|Default|GetFruit%20GameplayTag)))
  (return 0 (or _returnvalue_2 _returnvalue)))
