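; GA_ButterCake_PauseAIBehavior is a gameplay ability that, judging by its name, pauses ButterCake AI behavior for a fixed time.
; On activation it commits the ability, waits through a ten-second delay, then ends itself.
; The graph contains no explicit pause logic, and the end-ability event is empty.
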(event Ability|EventActivateAbility
  (Ability|CommitAbility)
  ;[1 1 1] SafeCheck
  (Utilities|FlowControl|Delay 1e+01)
  (Ability|EndAbility))

(event Ability|EventOnEndAbility
  (params (bWasCancelled :type Boolean)))
