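; Gameplay ability for ButterCake's AI ranged attack, firing arced projectile volleys at targets within a distance band.
; After resetting unstuck samples and pausing briefly, montage notifies trigger muzzle cues and shots until the maximum count jumps to Outro.
; Shots scatter around valid targets in the facing cone, otherwise toward random points ahead.
; SuggestProjectileVelocity solves each toss, choosing high arcs for steep, higher targets and retrying with an alternate launch speed.
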(event Ability|EventActivateAbility
  (Ability|CommitAbility)
  (Default|ResetUnstuckSampleLocation)
  (Utilities|FlowControl|Delay 1.0) ; We probably just executed a GroundPoundAttack or ChargeAttack before this attack,\nso just wait 1 sec to give time to ppl ejected to fall on ground.
  (Variables|Default|SetProjectilesCount 0)
  (CallFunction|StartFire))

(event Custom|StartFire
  (bind _montageinfo_montagetoplay (Utilities|Struct|BreakFortGameplayAbilityMontageInfo (Variables|GetMontageInfo_MontageToPlay)))
  ;[1 1 1] Play Montage
  (bind (_asynctaskproxy _notifytag) (Ability|Tasks|PlayMontageAndWaitNotify "BCRangedAttack" _montageinfo_montagetoplay)
    (:OnCompleted
      _)
    (:OnInterrupted
      _)
    (:OnCancelled
      _)
    (:OnNotifyBegin
      (switch Utilities|FlowControl|Switch|SwitchonGameplayTag _notifytag
        (:Event.ButterCake.RangedAttack
          (Utilities|FlowControl|Sequence
            (:then_0
              ;[1 1 1] Check if it's over (last bullet ?)
              (bind _result (Math|Integer|IncrementInt (Variables|Default|GetProjectilesCount))
                (:%20%20
                  (if (>= _result (ScalableFloat|GetValueatLevel (Variables|Default|GetMaxProjectilesCount)))
                    (Ability|Animation|MontageSetNextSectionName "Fire" "Outro")))))
            (:then_1
              ;[1 1 1] Fire !
              (bind _startlocation (CallFunction|ComputeStartLocation))
              (Variables|Default|SetLaunchLocation _startlocation)
              (bind _targetlocation (CallFunction|ComputeTargetLocation))
              (bind (_velocity _success) (CallFunction|ComputeVelocity (Variables|Default|GetLaunchLocation) _targetlocation))
              (if _success
                (CallFunction|SpawnProjectile (Variables|Default|GetLaunchLocation) _velocity)))))
        (:Event.ButterCake.RangedAttackMuzzleFlash
          ;[1 1 1] Aninotify : MuzzleFlash VFX
          (Ability|ExecuteGameplayCueOnOwner (Variables|Default|GetGameplayCueMuzzleVFX)))
        (:Event.ButterCake.RangedAttackMuzzleSound
          ;[1 1 1] Aninotify : MuzzleFlash Sound
          (Ability|ExecuteGameplayCueOnOwner (Variables|Default|GetGameplayCueMuzzleSFX))))))
  (Ability|EndAbility))

(fn ComputeStartLocation
  (return (+ (Transformation|GetSocketLocation (Actor|GetComponentbyClass (Ability|GetAvatarActorfromActorInfo) "/Script/FortniteGame.FortPooledSkelMeshComponent") (Variables|Default|GetOriginSocketName)) (* (Math|Vector|GetUpVector (Transformation|GetSocketRotation (Actor|GetComponentbyClass (Ability|GetAvatarActorfromActorInfo) "/Script/FortniteGame.FortPooledSkelMeshComponent") (Variables|Default|GetOriginSocketName))) (ScalableFloat|GetValueatLevel (LocalVariables|GetSpawnProjectileZOffset))))))

(fn ComputeTargetLocation
  (bind _returnvalue (Ability|GetAIAbilityTarget))
  (bind _returnvalue_2 (Transformation|GetActorLocation _returnvalue))
  (Utilities|IsValid _returnvalue
    (:Is%20Valid
      (bind _isvalid (CallFunction|IsValidTargetLocation _returnvalue_2))
      (if _isvalid
        (return (Utilities|Operators|Add _returnvalue_2 (Math|Vector|MakeVector (Math|Random|RandomFloatinRange (* (ScalableFloat|GetValueatLevel (LocalVariables|GetMaxRandomOffset)) -1.0) (ScalableFloat|GetValueatLevel (LocalVariables|GetMaxRandomOffset))) (Math|Random|RandomFloatinRange (* (ScalableFloat|GetValueatLevel (LocalVariables|GetMaxRandomOffset)) -1.0) (ScalableFloat|GetValueatLevel (LocalVariables|GetMaxRandomOffset))))))
        (else
          _)))
    (:Is%20Not%20Valid
      _))
  (return (+ (Transformation|GetActorLocation (Ability|GetAvatarActorfromActorInfo)) (* (Math|Vector|RotateVectorAroundAxis (Transformation|GetActorForwardVector (Ability|GetAvatarActorfromActorInfo)) (Math|Random|RandomFloatinRange -2e+01 2e+01) "0, 0,1.000000") (Math|Random|RandomFloatinRange 3e+03 5e+03)))))

(fn ComputeVelocity
  (params (StartLocation :type Vector) (EndLocation :type Vector))
  (bind (_startlocation_x _startlocation_y _startlocation_z) (Math|Vector|BreakVector (Variables|GetStartLocation_X)))
  (bind (_endlocation_x _endlocation_y _endlocation_z) (Math|Vector|BreakVector (Variables|GetEndLocation_X)))
  (LocalVariables|SetUseHighArc (and (> (Class|AbilityFunctions|GetAngleBetweenTwoVectors (Transformation|GetActorForwardVector (Ability|GetAvatarActorfromActorInfo)) (Math|Vector|Normalize (- (Variables|GetEndLocation) (Variables|GetStartLocation)))) (ScalableFloat|GetValueatLevel (LocalVariables|GetHighArcMaxAngle))) (< _startlocation_z _endlocation_z)))
  (bind (_tossvelocity _returnvalue) (Game|SuggestProjectileVelocity (Variables|GetStartLocation) (Variables|GetEndLocation) (ScalableFloat|GetValueatLevel (Variables|Default|GetProjectileLaunchSpeed)) 0.0 "DoNotTrace" 0.0 (LocalVariables|GetUseHighArc)))
  (if _returnvalue
    (return _tossvelocity _returnvalue)
    (else
      (bind (_tossvelocity_2 _returnvalue_2) (Game|SuggestProjectileVelocity (Variables|GetStartLocation) (Variables|GetEndLocation) (ScalableFloat|GetValueatLevel (Variables|Default|GetProjectileLaunchSpeed) 2.0) 0.0 "DoNotTrace" 0.0 (LocalVariables|GetUseHighArc)))
      (return _tossvelocity_2 _returnvalue_2))))

(fn DrawDebug
  (bind _maxangle (Variables|Default|GetMaxAngle))
  (bind _returnvalue (ScalableFloat|GetValueatLevel _maxangle))
  (bind _returnvalue_2 (Ability|GetAvatarActorfromActorInfo))
  (LocalVariables|SetActorLocation (Transformation|GetActorLocation _returnvalue_2))
  (LocalVariables|SetActorForward (Transformation|GetActorForwardVector _returnvalue_2))
  (Rendering|Debug|DrawDebugLine (LocalVariables|GetActorLocation) (+ (LocalVariables|GetActorLocation) (* (Math|Vector|RotateVector (LocalVariables|GetActorForward) (Math|Vector|RotatorfromAxisandAngle "0, 0,1.000000" _returnvalue)) 2e+03)) "(R=0.000000,G=1.000000,B=0.000000,A=1.000000)" 5.0 1e+01)
  (Rendering|Debug|DrawDebugLine (LocalVariables|GetActorLocation) (+ (LocalVariables|GetActorLocation) (* (Math|Vector|RotateVector (LocalVariables|GetActorForward) (Math|Vector|RotatorfromAxisandAngle "0, 0,1.000000" (* _returnvalue -1.0))) 2e+03)) "(R=0.000000,G=1.000000,B=0.000000,A=1.000000)" 5.0 1e+01)
  (Rendering|Debug|DrawDebugSphere (Transformation|GetActorLocation (Ability|GetAIAbilityTarget)) 1e+02 12 "(R=1.000000,G=0.000000,B=0.000000,A=1.000000)" 5.0 3.0))

(fn IsValidTargetLocation
  (params (TargetLocation :type Vector))
  (if (AI|IsValidAILocation TargetLocation)
    (if (> (Math|Vector|Distance(Vector) (Variables|GetTargetLocation) (Transformation|GetActorLocation (Ability|GetAvatarActorfromActorInfo))) (ScalableFloat|GetValueatLevel (Variables|Default|GetMinDistance)))
      (return (<= (Math|Float|Absolute(Float) (Class|AbilityFunctions|GetAngleBetweenTwoVectors (Math|Vector|Normalize2D(Vector) (- (Variables|GetTargetLocation) (Transformation|GetActorLocation (Ability|GetAvatarActorfromActorInfo)))) (Transformation|GetActorForwardVector (Ability|GetAvatarActorfromActorInfo)))) (ScalableFloat|GetValueatLevel (Variables|Default|GetMaxAngle))))
      (else
        _))
    (else
      _))
  (return false))

(fn CanActivateAbility
  (params (ActorInfo :type Gameplay%20Ability%20Actor%20Info%20Structure) (Handle :type Gameplay%20Ability%20Spec%20Handle%20Structure))
  (bind _returnvalue (Ability|GetAIAbilityTarget))
  (Utilities|IsValid _returnvalue
    (:Is%20Valid
      (return 0 (and (> (Math|Vector|Distance(Vector) (Transformation|GetActorLocation _returnvalue) (Transformation|GetActorLocation (Ability|GetAvatarActorfromActorInfo))) (ScalableFloat|GetValueatLevel (Variables|Default|GetMinDistance))) (< (Math|Vector|Distance(Vector) (Transformation|GetActorLocation _returnvalue) (Transformation|GetActorLocation (Ability|GetAvatarActorfromActorInfo))) (ScalableFloat|GetValueatLevel (Variables|Default|GetMaxDistance))))))
    (:Is%20Not%20Valid
      (return 0 false))))

(fn ResetUnstuckSampleLocation
  (bind _returnvalue (Ability|GetAvatarActorfromActorInfo))
  (bind _returnvalue_2 (Actor|GetComponentbyClass _returnvalue "/Script/ButterCakeRuntime.ButterCakeUnstuckComponent"))
  (Utilities|IsValid _returnvalue_2
    (:Is%20Valid
      (ButterCakeUnstuck|ResetUnstuckLocationSamples _returnvalue_2))))

(fn SpawnProjectile
  (params (SpawnLocation :type Vector) (Velocity :type Vector))
  (bind _returnvalue (Projectile|SpawnProjectile :ProjectileClass (Variables|Default|GetProjectile%20Class) :RequestedBy (Ability|GetActivatingPawn) :SpawnLocation (Variables|GetSpawnLocation) :SpawnRotation "0, 0, 0" :SpawnDirection (Math|Conversions|RotationFromXVector (Variables|GetVelocity)) :EffectContainerSpecToApplyOnExplode (Ability|MakeGameplayEffectContainerSpec (Variables|Default|GetT_Proj_HitApplicationTag) (Variables|Default|GetProj_HitEffectLevel)) :bUseDefaultPhysics false :InitialSpeed (Math|Vector|VectorLength (Variables|GetVelocity)) :GravityScale 1.0 :ChargePercent 0.0))
  (return _returnvalue))
