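; Gameplay ability that makes the ButterCake NPC sneeze out either chickens or loot from its blowhole, gated by the pawn's CanSneeze.
; A montage notify rolls for a chicken, allowed only outside Creative/Playground, below a rarity threshold and within Living World Manager event budget.
; Otherwise it launches loot-table pickups, their count scaled by highest vacuumed rarity, and tags them.
; Both paths notify the pawn, report telemetry and use short delays before cueing spawned objects.
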(event Custom|OnChickenSpawned
  (params (SpawnedActors :type Array%20of%20Actor%20Object%20References) (bSuccess :type Boolean))
  ;[1 1 1] Spawn Chicken
  (bind _output_get (Variables|Default|SetSpawned%20Actors SpawnedActors))
  (for _array_element _output_get
    (bind _asfort_aipawn (Utilities|Casting|CastToFortAIPawn _array_element))
    (Character|LaunchCharacter _asfort_aipawn (* (CallFunction|GetLaunchDirection) (ScalableFloat|GetValueatLevel (Variables|Default|GetNugVelocity))) true true))
  ;[1 1 1] Execute gameplay cue on nugs (currently for audio)
  (Utilities|FlowControl|Delay 0.1) ; Small delay to allow client version to spawn
  (for _array_element_2 (Variables|Default|GetSpawned%20Actors)
    (Utilities|IsValid _array_element_2
      (:Is%20Valid
        (GameplayCue|ExecuteGameplayCue _array_element_2 (Variables|Default|GetGCN_NugSneezed))))))

(event Ability|EventActivateAbility
  (bind _montageinfo_montagetoplay (Utilities|Struct|BreakFortGameplayAbilityMontageInfo (Variables|GetMontageInfo_MontageToPlay)))
  (bind _living_world_manager (Variables|Default|GetLiving%20World%20Manager))
  (Ability|CommitAbility)
  (Utilities|Array|Clear (Variables|Default|GetSpawnedItems))
  (bind _asnpc_pawn_butter_cake_base (Utilities|Casting|CastToNPC_Pawn_ButterCake_Base (Ability|GetAvatarActorfromActorInfo))
    (:then
      (Variables|Default|SetNPC%20Pawn%20Butter%20Cake%20Base _asnpc_pawn_butter_cake_base)
      (bind _livingworldmanager (CallFunction|GetLivingWorldManager))
      (Variables|Default|SetLiving%20World%20Manager _livingworldmanager)
      (Ability|Tasks|PlayMontageAndWaitNotify "MontageEatPickups" _montageinfo_montagetoplay
        (:OnCompleted
          _)
        (:OnInterrupted
          _)
        (:OnCancelled
          _)
        (:OnNotifyBegin
          (Variables|Default|SetBlowHoleLocation (Transformation|GetSocketLocation (Actor|GetComponentbyClass (Ability|GetAvatarActorfromActorInfo) "/Script/FortniteGame.FortPooledSkelMeshComponent") (Variables|Default|GetOriginSocketName)))
          (if (CallFunction|CanSpawnChicken)
            ;[1 1 1] Spawn Chicken
            (Ability|ExecuteGameplayCueWithParamsOnOwner (Variables|Default|GetGameplay%20Cue%20Tag%20Chiken) (Ability|GameplayCue|MakeGameplayCueParameters :NormalizedMagnitude 0.0 :RawMagnitude 0.0 :Location (Variables|Default|GetBlowHoleLocation) :Normal "0,0,0" :GameplayEffectLevel 1 :AbilityLevel 1 :bReplicateLocationWhenUsingMinimalRepProxy false))
            (Utilities|IsValid _living_world_manager
              (:Is%20Valid
                (Class|FortAthenaLivingWorldManager|TrySpawnEvent _living_world_manager (Variables|Default|GetLWM%20Chicken%20Event%20Entry) (Math|Conversions|ToTransform(Vector) (+ (Variables|Default|GetBlowHoleLocation) "0,0,100.000000")) (delegate Custom|OnChickenSpawned))
                (Sneeze|OnSneeze (Variables|Default|GetNPC%20Pawn%20Butter%20Cake%20Base))
                (CallFunction|FeedAnalytics 1)))
            (else
              ;[1 1 1] Spawn Loots
              (Ability|ExecuteGameplayCueWithParamsOnOwner (Variables|Default|GetGameplay%20Cue%20Tag%20Loot) (Ability|GameplayCue|MakeGameplayCueParameters :NormalizedMagnitude 0.0 :RawMagnitude 0.0 :Location (Variables|Default|GetBlowHoleLocation) :Normal "0,0,0" :GameplayEffectLevel 1 :AbilityLevel 1 :bReplicateLocationWhenUsingMinimalRepProxy false))
              (for _ (range (+ (CallFunction|GetRollLootMaxIndex) 1))
                (bind _returnvalue (Inventory|SpawnPickupInWorldWithLootTier (Variables|Default|GetLootTableName) (+ (Variables|Default|GetBlowHoleLocation) "0,0,100.000000") 0 0 true true (* (CallFunction|GetLaunchDirection) 3e+03)))
                (Utilities|Array|AppendArray (Variables|Default|GetSpawnedItems) _returnvalue))
              ;;[1 1 1] Apply a GameplayTag on Loots Sneezed
              (bind _returnvalue_2 (AI|GetAIServiceLoot))
              (for _array_element (Variables|Default|GetSpawnedItems)
                (AI|K2AddGameplayTagtoLoot _returnvalue_2 _array_element (Variables|Default|GetGameplayTagToApplyOnLoot)))
              (Utilities|FlowControl|Sequence
                (:then_0
                  ;;[1 1 1] On Sneeze
                  (Sneeze|OnSneeze (Variables|Default|GetNPC%20Pawn%20Butter%20Cake%20Base))
                  (CallFunction|FeedAnalytics (Utilities|Array|Length (Variables|Default|GetSpawnedItems))))
                (:then_1
                  ;;[1 1 1] Apply gameplay cue/trail cue to spawned loot objects. Delayed to increase chance of the actor being valid.
                  (Utilities|FlowControl|Delay 0.15)
                  (for _array_element_2 (Variables|Default|GetSpawnedItems)
                    (Ability|ExecuteGameplayCueWithParamsOnOwner (Variables|Default|GetGameplayCueSneezeLoot) (Ability|GameplayCue|MakeGameplayCueParameters :NormalizedMagnitude 0.0 :RawMagnitude 0.0 :Location "0,0,0" :Normal "0,0,0" :SourceObject _array_element_2 :GameplayEffectLevel 1 :AbilityLevel 1 :bReplicateLocationWhenUsingMinimalRepProxy false)))
                  (Utilities|Array|Clear (Variables|Default|GetSpawnedItems))))))))
      (Ability|EndAbility))
    (:CastFailed
      (Ability|EndAbility))))

(fn CanSpawnChicken
  (bind _living_world_manager (Variables|Default|GetLiving%20World%20Manager))
  (if (Fortnite|IsCreativeorPlaygroundModePlaylist)
    _
    (elif (>= (Math|Conversions|ToInteger(Byte) (Variables|/ButterCake/Pawns/NPC_Pawn_ButterCake_Base.NPC_Pawn_ButterCake_Base_C|GetCurrentMaxRarityItemsVacuumed (Variables|Default|GetNPC%20Pawn%20Butter%20Cake%20Base))) (ScalableFloat|GetValueasInteger (LocalVariables|GetRarityThreshold)))
      _
      (else
        (Utilities|IsValid _living_world_manager
          (:Is%20Valid
            (if (and (<= (Math|Random|RandomFloat) (ScalableFloat|GetValueatLevel (LocalVariables|GetChanceToSpawnChicken))) (> (ScalableFloat|GetValueatLevel (LocalVariables|GetChanceToSpawnChicken)) 0))
              (return (> (Class|FortAthenaLivingWorldManager|QueryEventBudget _living_world_manager (Variables|Default|GetLWM%20Chicken%20Event%20Entry) (Ability|GetAvatarActorfromActorInfo)) 0))
              (else
                _)))
          (:Is%20Not%20Valid
            _))
        (return false)))))

(fn FeedAnalytics
  (params (SneezedObjectCount :type Integer))
  (bind _asfort_aipawn (Utilities|Casting|CastToFortAIPawn (Ability|GetAvatarActorfromActorInfo)))
  (bind _asbutter_cake_controller (Utilities|Casting|CastToButterCake_Controller (AI|GetAIController _asfort_aipawn)))
  (ButterCake-Telemetry|OnItemsSneezed (Variables|/ButterCake/Controllers/ButterCake_Controller.ButterCake_Controller_C|GetFortButterCakeComponent_Telemetry _asbutter_cake_controller) (Variables|GetSneezedObjectCount)))

(fn GetRollLootMaxIndex
  (bind _npc_pawn_butter_cake_base (Variables|Default|GetNPC%20Pawn%20Butter%20Cake%20Base))
  (bind _currentmaxrarityitemsvacuumed (Variables|/ButterCake/Pawns/NPC_Pawn_ButterCake_Base.NPC_Pawn_ButterCake_Base_C|GetCurrentMaxRarityItemsVacuumed _npc_pawn_butter_cake_base))
  (bind _returnvalue (Math|Conversions|ToInteger(Byte) _currentmaxrarityitemsvacuumed))
  (if (>= _returnvalue (ScalableFloat|GetValueasInteger (Variables|Default|GetRarityThresholdLevel2)))
    (LocalVariables|SetRoolCount (ScalableFloat|GetValueasInteger (Variables|Default|GetRarityLootCountLevel2)))
    (elif (>= _returnvalue (ScalableFloat|GetValueasInteger (Variables|Default|GetRarityThresholdLevel1)))
      (LocalVariables|SetRoolCount (ScalableFloat|GetValueasInteger (Variables|Default|GetRarityLootCountLevel1)))
      (else
        (LocalVariables|SetRoolCount 1))))
  (return (Math|Integer|Max(Integer) (- (LocalVariables|GetRoolCount) 1))))

(fn GetLaunchDirection
  (return (Math|Vector|RotateVectorAroundAxis (Math|Vector|VectorUp) (Math|Random|RandomFloatinRange (* (ScalableFloat|GetValueatLevel (Variables|Default|GetSneezeAngle)) -1.0) (ScalableFloat|GetValueatLevel (Variables|Default|GetSneezeAngle))) (Transformation|GetActorForwardVector (Ability|GetAvatarActorfromActorInfo)))))

(fn GetLivingWorldManager
  (bind _returnvalue (Ability|GetAvatarActorfromActorInfo))
  (bind _returnvalue_2 (Fortnite|Athena|GameState|GetGameStateAthenafromContext _returnvalue))
  (bind _returnvalue_3 (Actor|GetComponentbyClass _returnvalue_2 "/Script/LagerRuntime.FortAthenaLivingWorldManager"))
  (Utilities|IsValid _returnvalue_2
    (:Is%20Valid
      (Utilities|IsValid _returnvalue_3
        (:Is%20Valid
          (return _returnvalue_3))
        (:Is%20Not%20Valid
          _)))
    (:Is%20Not%20Valid
      _))
  (return 0))

(fn CanActivateAbility
  (params (ActorInfo :type Gameplay%20Ability%20Actor%20Info%20Structure) (Handle :type Gameplay%20Ability%20Spec%20Handle%20Structure))
  (bind _asnpc_pawn_butter_cake_base (Utilities|Casting|CastToNPC_Pawn_ButterCake_Base (Utilities|Struct|BreakGameplayAbilityActorInfo ActorInfo))
    (:then
      (bind _result (Sneeze|CanSneeze _asnpc_pawn_butter_cake_base))
      (return 0 _result))
    (:CastFailed
      (return 0 false))))
