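; Gameplay ability in which the ButterCake AI eats a target bush and vacuums up nearby building actors.
; A montage notify applies a destroy effect to the bush and gathers building actors in a sphere around a mesh socket.
; A next-tick timer processes one actor per frame for performance, applying building damage and optionally pulling physics objects toward the socket with spin.
; Ability end processes any actors still queued in one pass.
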(event Ability|EventActivateAbility
  (bind (_montageinfo_montagetoplay _montageinfo_animplayrate _montageinfo_animrootmotiontranslationscale _montageinfo_montagesectiontoplay _montageinfo_overridesection) (Utilities|Struct|BreakFortGameplayAbilityMontageInfo (Variables|GetMontageInfo_MontageToPlay)))
  (Ability|CommitAbility)
  (Utilities|Array|Clear (Variables|Default|GetBuildingActors))
  ;[1 1 1] Play Eat buch animation and wait for Gameplay event (from anim)
  (Ability|Tasks|PlayMontageAndWaitNotify "MontageEatBush" _montageinfo_montagetoplay _montageinfo_animplayrate _montageinfo_overridesection true _montageinfo_animrootmotiontranslationscale
    (:OnCompleted
      _)
    (:OnInterrupted
      _)
    (:OnCancelled
      _)
    (:OnNotifyBegin
      ;[1 1 1] Destroy Bush
      (bind _asbuilding_actor (Utilities|Casting|CastToBuildingActor (Ability|GetAIAbilityTarget))
        (:then
          (Ability|ApplyGameplayEffectToTarget (Ability|TargetData|AbilityTargetDatafromActor _asbuilding_actor) (Variables|Default|GetDestroyBuilding_GE)))
        (:CastFailed
          _))
      ;[1 1 1] Destroy Building Actors and Activate Phys Objects
      (bind _outactors (Collision|SphereOverlapActors (Default|GetVacuumLocation) (ScalableFloat|GetValueatLevel (Variables|Default|GetVacuumRadius)) (Variables|Default|GetBuilding%20Object%20Types) "/Script/FortniteGame.BuildingActor"))
      (Variables|Default|SetBuildingActors _outactors)
      (Utilities|Time|SetTimerforNextTickbyEvent (delegate Custom|ProcessBuidlingActors))))
  (Ability|EndAbility))

(event Custom|ProcessBuidlingActors
  (bind _returnvalue (Utilities|Array|Length (Variables|Default|GetBuildingActors)))
  (bind _returnvalue_2 (- _returnvalue 1))
  (bind _buildingactors (Variables|Default|GetBuildingActors))
  ;[1 1 1] Process BuildingActor 1 by Frame for Perf
  (if (Utilities|Array|IsNotEmpty _buildingactors)
    (Default|VacuumBuildingActor (Utilities|Array|Get(acopy) (Variables|Default|GetBuildingActors) _returnvalue_2))
    (Utilities|Array|RemoveIndex (Variables|Default|GetBuildingActors) _returnvalue_2)
    (Utilities|Time|SetTimerforNextTickbyEvent (delegate Custom|ProcessBuidlingActors))))

(event Ability|EventOnEndAbility
  (params (bWasCancelled :type Boolean))
  (for _array_element (Variables|Default|GetBuildingActors)
    (Default|VacuumBuildingActor _array_element)))

(fn GetVacuumLocation
  (return (Transformation|GetSocketLocation (Actor|GetComponentbyClass (Ability|GetAvatarActorfromActorInfo) "/Script/FortniteGame.FortPooledSkelMeshComponent") (Variables|Default|GetSocketNameForVacuum))))

(fn VacuumBuildingActor
  (params (BuildingActor :type Actor%20Object%20Reference))
  (Utilities|IsValid BuildingActor
    (:Is%20Valid
      (if (ScalableFloat|GetValueasBool (Variables|Default|GetPhysObjectEnable))
        (bind _returnvalue (FortPhysicsSimulationLibrary|CanActorSimulatePhysics (Variables|GetBuildingActor)))
        (if _returnvalue
          (Default|VacuumPhysObject (Variables|GetBuildingActor))
          (else
            _)))
      (Ability|ApplyGameplayEffectToTarget (Ability|TargetData|AbilityTargetDatafromActor (Variables|GetBuildingActor)) "/ButterCake/Abilities/GE_ButterCake_Vacuum_DmgBuilding.GE_ButterCake_Vacuum_DmgBuilding_C"))))

(fn VacuumPhysObject
  (params (PhysActor :type Actor%20Object%20Reference))
  (PhysicsObject|WakeUpPhysicsObject PhysActor)
  (FortPhysicsSimulationLibrary|SetPhysicsLinearVelocity PhysActor (* (Math|Vector|Normalize (- (Default|GetVacuumLocation) (Transformation|GetActorLocation PhysActor))) (ScalableFloat|GetValueatLevel (LocalVariables|GetPhysObjectVelocity))) false "None" 0 "(MaxSimSizeAllowed=Large,bRejectEmbeddedInLandscape=False)")
  (FortPhysicsSimulationLibrary|SetPhysicsAngularVelocity PhysActor "180.000000,180.000000,180.000000" false "None" 0 "(MaxSimSizeAllowed=Large,bRejectEmbeddedInLandscape=False)"))
