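; Gameplay ability where the ButterCake AI creature vacuums up its target pickup and others nearby while playing a montage.
; On montage notify, sphere overlaps gather building actors, damaged or pulled as physics objects one per frame, then pickups, which it grabs.
; Eaten fruit triggers a cue and, at blend-out, Berserk removal; crowns add a cue at ability end.
; Item count and top rarity feed the pawn, count feeds telemetry, and leftover buildings are processed at end.
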(event Ability|EventActivateAbility
  (bind (_montageinfo_montagetoplay _montageinfo_animplayrate _montageinfo_animrootmotiontranslationscale _montageinfo_montagesectiontoplay _montageinfo_overridesection) (Utilities|Struct|BreakFortGameplayAbilityMontageInfo (Variables|GetMontageInfo_MontageToPlay)))
  (Ability|CommitAbility)
  (CallFunction|ResetUnstuckSampleLocation)
  (Variables|Default|SetHasEatenFruit false)
  (Utilities|Array|Clear (Variables|Default|GetBuildingActors))
  (bind _asfort_pickup (Utilities|Casting|CastToFortPickup (Ability|GetAIAbilityTarget))
    (:then
      (Variables|Default|SetActorToPickup _asfort_pickup)
      (bind _asfort_aipawn (Utilities|Casting|CastToFortAIPawn (Ability|GetAvatarActorfromActorInfo))
        (:then
          (bind _output_get (Variables|Default|SetAIPawn _asfort_aipawn))
          (Variables|Default|SetCollisionReactionComp (Actor|GetComponentbyClass _output_get "/Script/FortniteGame.FortAthenaCollisionReactionComponent"))
          ;[1 1 1] Play Vacuum animation and wait for Gameplay event (from anim)
          (Ability|Tasks|PlayMontageAndWaitNotify "MontageEatPickups" _montageinfo_montagetoplay _montageinfo_animplayrate _montageinfo_overridesection true _montageinfo_animrootmotiontranslationscale
            (:OnCompleted
              ;[1 1 1] Waiting for pickup destroy before ending ability
              (Utilities|FlowControl|Delay 1.0)
              (Ability|EndAbility))
            (:OnBlendOut
              ;[1 1 1] Remove Berserk if needed - Triggered OnBlendOut to be call just before the Complet to avoid BerserkFace Glitchs
              (if (Variables|Default|GetHasEatenFruit)
                (Ability|ApplyGameplayEffectToOwner "/ButterCake/Abilities/GE_ButterCake_JustEatenFruit.GE_ButterCake_JustEatenFruit_C")
                (Ability|RemoveGameplayEffectFromOwnerWithAssetTags "(GameplayTags=((TagName=\"Athena.Gameplay.Status.Berserk\")))")))
            (:OnInterrupted
              _)
            (:OnCancelled
              _)
            (:OnNotifyBegin
              ;[1 1 1] Destroy Building Actors and Activate Phys Objects
              (bind _outactors (Collision|SphereOverlapActors (CallFunction|GetVacuumLocation) (ScalableFloat|GetValueatLevel (Variables|Default|GetVacuumRadius)) (Variables|Default|GetBuilding%20Object%20Types) "/Script/FortniteGame.BuildingActor"))
              (Variables|Default|SetBuildingActors _outactors)
              (Utilities|Time|SetTimerforNextTickbyEvent (delegate Custom|ProcessBuidlingActors))
              (Utilities|FlowControl|Delay 0.2)
              ;[1 1 1] Find Pickup Objects to Vacuum
              (bind _outactors_2 (Collision|SphereOverlapActors (CallFunction|GetVacuumLocation) (ScalableFloat|GetValueatLevel (Variables|Default|GetVacuumRadius)) (Variables|Default|GetLoots%20Object%20Types) "/Script/FortniteGame.FortPickup" (Utilities|Array|MakeArray (Variables|Default|GetActorToPickup))))
              ;[1 1 1] Vacuum
              (Utilities|Array|Insert _outactors_2 (Variables|Default|GetActorToPickup))
              (bind (_haseatenfruit _haseatenvictorycrown _raritymax _vacuumitemcount) (CallFunction|VacuumItems _outactors_2))
              (Variables|Default|SetHasEatenFruit _haseatenfruit)
              (Variables|Default|SetHas%20Eaten%20Victory%20Crown _haseatenvictorycrown)
              (Variables|Default|SetRarity%20Max _raritymax)
              (Variables|Default|SetVacuum%20Item%20Count _vacuumitemcount)
              (Utilities|FlowControl|Sequence
                (:then_0
                  ;[1 1 1] Just eat a fruit ?
                  (if (Variables|Default|GetHasEatenFruit)
                    (Ability|ExecuteGameplayCueOnOwner (Variables|Default|GetEat%20Fruit%20Gameplay%20Cue%20Tag))))
                (:then_1
                  ;[1 1 1] Store the vacuumed pickup count on the pawn
                  (bind _asnpc_pawn_butter_cake_base (Utilities|Casting|CastToNPC_Pawn_ButterCake_Base (Variables|Default|GetAIPawn)))
                  (Vacuum|OnVacuumPickup _asnpc_pawn_butter_cake_base (Variables|Default|GetVacuum%20Item%20Count) (Variables|Default|GetRarity%20Max)))
                (:then_2
                  ;[1 1 1] Analytics
                  (CallFunction|FeedAnalytics (Variables|Default|GetVacuum%20Item%20Count))))))
          (Ability|EndAbility))
        (:CastFailed
          _)))
    (:CastFailed
      _))
  (Ability|EndAbility))

(event Ability|EventOnEndAbility
  (params (bWasCancelled :type Boolean))
  (Utilities|FlowControl|Sequence
    (:then_0
      (for _array_element (Variables|Default|GetBuildingActors)
        (CallFunction|VacuumBuildingActor _array_element)))
    (:then_1
      ;[1 1 1] Just eat a VictoryCrown ?
      (if (Variables|Default|GetHas%20Eaten%20Victory%20Crown)
        (Ability|AddGameplayCueToOwner :GameplayCueTag (Variables|Default|GetEat%20Crown%20Gameplay%20Cue%20Tag) :bRemoveOnAbilityEnd false)))))

(event Custom|ProcessBuidlingActors
  (bind _returnvalue (Utilities|Array|Length (Variables|Default|GetBuildingActors)))
  (bind _returnvalue_2 (- _returnvalue 1))
  (bind _buildingactors (Variables|Default|GetBuildingActors))
  ;[1 1 1] Process BuildingActor 1 by Frame for Perf
  (if (Utilities|Array|IsNotEmpty _buildingactors)
    (CallFunction|VacuumBuildingActor (Utilities|Array|Get(acopy) (Variables|Default|GetBuildingActors) _returnvalue_2))
    (Utilities|Array|RemoveIndex (Variables|Default|GetBuildingActors) _returnvalue_2)
    (Utilities|Time|SetTimerforNextTickbyEvent (delegate Custom|ProcessBuidlingActors))))

(fn FeedAnalytics
  (params (ItemCount :type Integer))
  (bind _asbutter_cake_controller (Utilities|Casting|CastToButterCake_Controller (AI|GetAIController (Variables|Default|GetAIPawn))))
  (ButterCake-Telemetry|OnItemsEaten (Variables|/ButterCake/Controllers/ButterCake_Controller.ButterCake_Controller_C|GetFortButterCakeComponent_Telemetry _asbutter_cake_controller) (Variables|GetItemCount)))

(fn GetVacuumLocation
  (return (Transformation|GetSocketLocation (Actor|GetComponentbyClass (Ability|GetAvatarActorfromActorInfo) "/Script/FortniteGame.FortPooledSkelMeshComponent") (Variables|Default|GetSocketNameForVacuum))))

(fn ResetUnstuckSampleLocation
  (bind _returnvalue (Ability|GetAvatarActorfromActorInfo))
  (bind _returnvalue_2 (Actor|GetComponentbyClass _returnvalue "/Script/ButterCakeRuntime.ButterCakeUnstuckComponent"))
  (Utilities|IsValid _returnvalue_2
    (:Is%20Valid
      (ButterCakeUnstuck|ResetUnstuckLocationSamples _returnvalue_2))))

(fn VacuumBuildingActor
  (params (BuildingActor :type Actor%20Object%20Reference))
  (bind _collisionreactioncomp (Variables|Default|GetCollisionReactionComp))
  (Utilities|IsValid BuildingActor
    (:Is%20Valid
      ;[1 1 1] Make sure to ignore building actors that are not valid for our reactions (damage or physics activation)
      (bind _asbuilding_actor (Utilities|Casting|CastToBuildingActor BuildingActor)
        (:then
          (Utilities|IsValid _collisionreactioncomp
            (:Is%20Valid
              (if (CollisionReactions|K2IsBuildingActorValidforReaction _collisionreactioncomp _asbuilding_actor)
                _))
            (:Is%20Not%20Valid
              _))
          (if (ScalableFloat|GetValueasBool (Variables|Default|GetPhysObjectEnable))
            (bind _returnvalue (FortPhysicsSimulationLibrary|CanActorSimulatePhysics (Variables|GetBuildingActor)))
            (if _returnvalue
              (CallFunction|VacuumPhysObject (Variables|GetBuildingActor))
              (else
                _)))
          (Ability|ApplyGameplayEffectToTarget (Ability|TargetData|AbilityTargetDatafromActor (Variables|GetBuildingActor)) "/ButterCake/Abilities/GE_ButterCake_Vacuum_DmgBuilding.GE_ButterCake_Vacuum_DmgBuilding_C"))
        (:CastFailed
          _)))))

(fn VacuumItems
  (params (Actors :type Array%20of%20Actor%20Object%20References))
  (bind _pickupitemdefinition (LocalVariables|GetPickupItemDefinition))
  (bind _returnvalue (Class|FortItemFunctionLibrary|GetRarity _pickupitemdefinition))
  ;[1 1 1] Init
  (LocalVariables|SetItemCount 0)
  (LocalVariables|SetEatenVictoryCrown false)
  (LocalVariables|SetEatenFruit false)
  (LocalVariables|SetEatenRarityMax "Common")
  (bind _array_element (Utilities|Array|ForEachLoopwithBreak Actors)
    (:LoopBody
      (bind _asfort_pickup (Utilities|Casting|CastToFortPickup _array_element))
      (bind _output_get (LocalVariables|SetPickup _asfort_pickup))
      (if (not (Pickup|PickedUp _asfort_pickup))
        (Utilities|IsValid (Pickup|GetPrimaryItemDefinition _output_get)
          (:Is%20Valid
            (LocalVariables|SetPickupItemDefinition (Pickup|GetPrimaryItemDefinition _output_get))
            (Utilities|FlowControl|Sequence
              (:then_0
                _)
              (:then_1
                ;[1 1 1] Check if it's a Crown
                (bind _asathena_gadget_item_definition (Utilities|Casting|CastToAthenaGadgetItemDefinition (LocalVariables|GetPickupItemDefinition)))
                (bind _itemdata (Class|ItemDefinitionBase|GetItemData _asathena_gadget_item_definition)
                  (:Found
                    (if (GameplayTags|MatchesTag (Utilities|Struct|BreakFortniteSpecialActorPickupItemData _itemdata) (Variables|Default|GetVictoryCrownTag) true)
                      (LocalVariables|SetEatenVictoryCrown true)))))
              (:then_2
                ;[1 1 1] Check the Rarity
                (if (> (Math|Conversions|ToInteger(Byte) _returnvalue) (Math|Conversions|ToInteger(Byte) (LocalVariables|GetEatenRarityMax)))
                  (LocalVariables|SetEatenRarityMax _returnvalue)))
              (:then_3
                ;[1 1 1] Grab the Pickup
                (AI|GrabPickupCustomized (Variables|Default|GetAIPawn) (LocalVariables|GetPickup) (ScalableFloat|GetValueatLevel (Variables|Default|GetVacuumPickupFlyTime))))
              (:then_4
                ;[1 1 1] Quest
                (Class|FortPickup|NotifyQuestManager (LocalVariables|GetPickup) "Feed" (GameplayTags|GetOwnedGameplayTags (Variables|Default|GetAIPawn) :!class /Script/GameplayTags.GameplayTagAssetInterface)))
              (:then_5
                ;[1 1 1] Count Items
                (bind _result (Math|Integer|IncrementInt (LocalVariables|GetItemCount))
                  (:%20%20
                    (if (>= _result (ScalableFloat|GetValueasInteger (LocalVariables|GetMaxObjectCountByVacuum)))
                      _)))))
            ;[1 1 1] Check if it's a BigFruit
            (bind _returnvalue_2 (GameplayTags|HasAnyMatchingGameplayTags(Message) (LocalVariables|GetPickupItemDefinition) (Variables|Default|GetFruitTags)))
            (if _returnvalue_2
              (LocalVariables|SetEatenFruit true))))))
    (:Completed
      _))
  (return (LocalVariables|GetEatenFruit) (LocalVariables|GetEatenVictoryCrown) (LocalVariables|GetEatenRarityMax) (LocalVariables|GetItemCount)))

(fn VacuumPhysObject
  (params (PhysActor :type Actor%20Object%20Reference))
  (PhysicsObject|WakeUpPhysicsObject PhysActor)
  (FortPhysicsSimulationLibrary|SetPhysicsLinearVelocity PhysActor (* (Math|Vector|Normalize (- (CallFunction|GetVacuumLocation) (Transformation|GetActorLocation PhysActor))) (ScalableFloat|GetValueatLevel (LocalVariables|GetPhysObjectVelocity))) false "None" 0 "(MaxSimSizeAllowed=Large,bRejectEmbeddedInLandscape=False)")
  (FortPhysicsSimulationLibrary|SetPhysicsAngularVelocity PhysActor "180.000000,180.000000,180.000000" false "None" 0 "(MaxSimSizeAllowed=Large,bRejectEmbeddedInLandscape=False)"))
